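# Physics asset for the back jacket piece of the female medium Red Riding Remix player cosmetic body.
# A box body on the dyn_jacket_back bone covers the jacket flap, with capsule bodies on the pelvis and spine_04.
# The flap is constrained to spine_04 with a tight twist limit and an asymmetric swing, while spine_04 is linked to the pelvis by a looser, uniform constraint.
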PhysicsAsset /BRCosmetics/Characters/Player/Female/Medium/Bodies/F_MED_RedRidingRemix/Meshes/F_MED_Red_Riding_Remix_BackJacket_Physics
bodies {
 dyn_jacket_back [{shapeType=Box, extentX=8.9547, extentY=1.4482, extentZ=15.4491, center=<5.8872 0 0>}]
 pelvis [{shapeType=Capsule, length=10.9878, radius=7.4221, center=<-4.2512 -1.4626 -0.0004>, rotation=<rot 90 0 0>}]
 spine_04 [{shapeType=Capsule, length=6.0781, radius=8.3591, center=<-2.2477 0.9723 0>}]
}
constraints [
 {bone1Name=spine_04, bone2Name=pelvis, swing1LimitDegrees=45, swing2LimitDegrees=45, twistLimitDegrees=45}
 {bone1Name=dyn_jacket_back, bone2Name=spine_04, swing1LimitDegrees=45, swing2LimitDegrees=20, twistLimitDegrees=5}
]
