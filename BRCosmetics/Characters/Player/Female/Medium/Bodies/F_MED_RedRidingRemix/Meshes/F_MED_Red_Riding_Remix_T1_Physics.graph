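# Physics asset for the Red Riding Remix female medium body mesh, giving its jacket and hood bones collision and swing for secondary motion.
# Capsules cover pelvis, upper spine, neck and hood, while flat boxes cover the jacket's back and side panels.
# Jacket panels hang from spine_04, with left and right front panels as two-link chains, and the hood hangs from the neck.
# Angular limits constrain each joint, tightest on the center back panel.
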PhysicsAsset /BRCosmetics/Characters/Player/Female/Medium/Bodies/F_MED_RedRidingRemix/Meshes/F_MED_Red_Riding_Remix_T1_Physics
bodies {
 dyn_hood [{shapeType=Capsule, length=5, radius=3, center=<0 -0.5 0>}]
 dyn_jacket_back [{shapeType=Box, extentX=8.9547, extentY=1.4482, extentZ=15.4491, center=<5.8872 0 0>}]
 dyn_jacket_l_01 [{shapeType=Box, extentX=6.5849, extentY=0.6732, extentZ=9.8441, center=<3.6246 0.5379 -0.672>}]
 dyn_jacket_l_02 [{shapeType=Box, extentX=8.4181, extentY=2.9917, extentZ=9.8217, center=<2.0681 0.4574 -1.9665>}]
 dyn_jacket_l_back [{shapeType=Box, extentX=8.4765, extentY=3.7211, extentZ=8.517, center=<3.6062 -0.1198 0.5437>}]
 dyn_jacket_r_01 [{shapeType=Box, extentX=6.5849, extentY=0.6704, extentZ=9.8441, center=<3.6246 0.5368 0.672>}]
 dyn_jacket_r_02 [{shapeType=Box, extentX=8.4181, extentY=2.9917, extentZ=9.8217, center=<2.0681 0.4577 1.9665>}]
 dyn_jacket_r_back [{shapeType=Box, extentX=8.4765, extentY=4.3076, extentZ=8.0091, center=<3.6062 -0.1163 -0.5088>}]
 neck_01 [{shapeType=Capsule, length=0.505, radius=0.505}]
 pelvis [{shapeType=Capsule, length=10.9878, radius=7.4221, center=<-4.2512 -1.4626 -0.0004>, rotation=<rot 90 0 0>}]
 spine_04 [{shapeType=Capsule, length=6.0781, radius=8.3591, center=<-2.2477 0.5584 0>}]
}
constraints [
 {bone1Name=spine_04, bone2Name=pelvis, swing1LimitDegrees=45, swing2LimitDegrees=45, twistLimitDegrees=45}
 {bone1Name=dyn_jacket_l_back, bone2Name=spine_04, swing1LimitDegrees=45, swing2LimitDegrees=45, twistLimitDegrees=15}
 {bone1Name=dyn_jacket_l_01, bone2Name=spine_04, swing1LimitDegrees=45, swing2LimitDegrees=45, twistLimitDegrees=15}
 {bone1Name=dyn_jacket_l_02, bone2Name=dyn_jacket_l_01, swing1LimitDegrees=45, swing2LimitDegrees=45, twistLimitDegrees=15}
 {bone1Name=dyn_jacket_r_back, bone2Name=spine_04, swing1LimitDegrees=45, swing2LimitDegrees=45, twistLimitDegrees=15}
 {bone1Name=dyn_jacket_r_01, bone2Name=spine_04, swing1LimitDegrees=45, swing2LimitDegrees=45, twistLimitDegrees=15}
 {bone1Name=dyn_jacket_r_02, bone2Name=dyn_jacket_r_01, swing1LimitDegrees=45, swing2LimitDegrees=45, twistLimitDegrees=15}
 {bone1Name=dyn_jacket_back, bone2Name=spine_04, swing1LimitDegrees=45, swing2LimitDegrees=20, twistLimitDegrees=5}
 {bone1Name=neck_01, bone2Name=spine_04, swing1LimitDegrees=45, swing2LimitDegrees=45, twistLimitDegrees=15}
 {bone1Name=dyn_hood, bone2Name=neck_01, swing1LimitDegrees=45, swing2LimitDegrees=45, twistLimitDegrees=15}
]
