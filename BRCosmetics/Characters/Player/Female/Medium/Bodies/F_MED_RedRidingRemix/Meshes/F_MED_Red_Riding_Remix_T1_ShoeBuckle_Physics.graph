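# Physics asset for the shoe buckles on the female medium Red Riding Remix player cosmetic, letting the dyn_buckle bones move as secondary motion.
# Each leg has four thin box bodies for buckles, constrained in pairs to the two calf twist capsules with swing limits and a narrower twist limit.
# The calf capsules tie to a single pelvis capsule with free angular motion, so the rig covers only the bones the buckles need.
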PhysicsAsset /BRCosmetics/Characters/Player/Female/Medium/Bodies/F_MED_RedRidingRemix/Meshes/F_MED_Red_Riding_Remix_T1_ShoeBuckle_Physics
bodies {
 calf_twist_01_l [{shapeType=Capsule, length=6.9384, radius=2.6686, center=<-1.0521 -0.7886 -0.1189>, rotation=<rot 90 0 0>}]
 calf_twist_01_r [{shapeType=Capsule, length=6.9384, radius=2.6686, center=<1.0519 0.7886 0.1189>, rotation=<rot 90 0 0>}]
 calf_twist_02_l [{shapeType=Capsule, length=11.8162, radius=4.0585, center=<-3.9424 -2.3647 -0.4053>, rotation=<rot 90 0 0>}]
 calf_twist_02_r [{shapeType=Capsule, length=11.8162, radius=4.0585, center=<3.9423 2.3647 0.4052>, rotation=<rot 90 0 0>}]
 dyn_buckle_l_01 [{shapeType=Box, extentX=4, extentY=4, extentZ=1, center=<0 -1.0944 0>}]
 dyn_buckle_l_02 [{shapeType=Box, extentX=4, extentY=4, extentZ=1, center=<0 -1.1657 0>}]
 dyn_buckle_l_03 [{shapeType=Box, extentX=4, extentY=4, extentZ=1, center=<0 -1.084 0>}]
 dyn_buckle_l_04 [{shapeType=Box, extentX=4, extentY=4, extentZ=1, center=<0 -1.1088 0>}]
 dyn_buckle_r_01 [{shapeType=Box, extentX=4, extentY=4, extentZ=1, center=<0 1.0583 0>}]
 dyn_buckle_r_02 [{shapeType=Box, extentX=4, extentY=4, extentZ=1, center=<0 1.0957 0>}]
 dyn_buckle_r_03 [{shapeType=Box, extentX=4, extentY=4, extentZ=1, center=<0 1.1272 0>}]
 dyn_buckle_r_04 [{shapeType=Box, extentX=4, extentY=4, extentZ=1, center=<0 1 0>}]
 pelvis [{shapeType=Capsule, length=4.7374, radius=8.9166, center=<-7.5911 -1.6847 -0.0009>, rotation=<rot 90 0 0>}]
}
constraints [
 {bone1Name=dyn_buckle_r_01, bone2Name=calf_twist_02_r, swing1LimitDegrees=45, swing2LimitDegrees=45, twistLimitDegrees=15}
 {bone1Name=dyn_buckle_r_02, bone2Name=calf_twist_02_r, swing1LimitDegrees=45, swing2LimitDegrees=45, twistLimitDegrees=15}
 {bone1Name=dyn_buckle_r_04, bone2Name=calf_twist_01_r, swing1LimitDegrees=45, swing2LimitDegrees=45, twistLimitDegrees=15}
 {bone1Name=dyn_buckle_r_03, bone2Name=calf_twist_01_r, swing1LimitDegrees=45, swing2LimitDegrees=45, twistLimitDegrees=15}
 {bone1Name=dyn_buckle_l_01, bone2Name=calf_twist_02_l, swing1LimitDegrees=45, swing2LimitDegrees=45, twistLimitDegrees=15}
 {bone1Name=dyn_buckle_l_02, bone2Name=calf_twist_02_l, swing1LimitDegrees=45, swing2LimitDegrees=45, twistLimitDegrees=15}
 {bone1Name=dyn_buckle_l_03, bone2Name=calf_twist_01_l, swing1LimitDegrees=45, swing2LimitDegrees=45, twistLimitDegrees=15}
 {bone1Name=dyn_buckle_l_04, bone2Name=calf_twist_01_l, swing1LimitDegrees=45, swing2LimitDegrees=45, twistLimitDegrees=15}
 {
  bone1Name=calf_twist_01_l
  bone2Name=pelvis
  swing1LimitDegrees=45
  swing1Motion=Free
  swing2LimitDegrees=45
  swing2Motion=Free
  twistLimitDegrees=45
  twistMotion=Free
 }
 {
  bone1Name=calf_twist_02_l
  bone2Name=pelvis
  swing1LimitDegrees=45
  swing1Motion=Free
  swing2LimitDegrees=45
  swing2Motion=Free
  twistLimitDegrees=45
  twistMotion=Free
 }
 {
  bone1Name=calf_twist_02_r
  bone2Name=pelvis
  swing1LimitDegrees=45
  swing1Motion=Free
  swing2LimitDegrees=45
  swing2Motion=Free
  twistLimitDegrees=45
  twistMotion=Free
 }
 {
  bone1Name=calf_twist_01_r
  bone2Name=pelvis
  swing1LimitDegrees=45
  swing1Motion=Free
  swing2LimitDegrees=45
  swing2Motion=Free
  twistLimitDegrees=45
  twistMotion=Free
 }
]
