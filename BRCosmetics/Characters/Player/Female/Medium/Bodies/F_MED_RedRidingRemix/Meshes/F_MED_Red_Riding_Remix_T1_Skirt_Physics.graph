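# Physics asset for the skirt of the female medium Red Riding Remix player cosmetic, covering its two back skirt bones.
# Box bodies on the left and right skirt bones, the left with an extra box, sit alongside capsules on the pelvis and thighs.
# Each skirt bone constrains to the pelvis with angular limits.
# The thigh-to-pelvis constraints and the link between the two skirt bones leave swing and twist free.
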PhysicsAsset /BRCosmetics/Characters/Player/Female/Medium/Bodies/F_MED_RedRidingRemix/Meshes/F_MED_Red_Riding_Remix_T1_Skirt_Physics
bodies {
 dyn_skirt_back_l [
  {shapeType=Box, extentX=20.764, extentY=4.3753, extentZ=18.288, center=<10.1459 1.1565 -1.7487>}
  {shapeType=Box, extentX=7.5, extentY=13, extentZ=3, center=<14.7168 7.3116 11.2176>}
 ]
 dyn_skirt_back_r [{shapeType=Box, extentX=20.7374, extentY=4.2519, extentZ=15.8421, center=<10.1327 1.2175 0.5388>}]
 pelvis [{shapeType=Capsule, length=7.9874, radius=5.6666, center=<-7.5911 -1.6847 -0.0009>, rotation=<rot 90 0 0>}]
 thigh_l [{shapeType=Capsule, length=22.7654, radius=8.2416, center=<-22 0.6427 -1.0404>, rotation=<rot 90 0 0>}]
 thigh_r [{shapeType=Capsule, length=22.7654, radius=8.2412, center=<22 -0.6425 0.8772>, rotation=<rot 90 0 0>}]
}
constraints [
 {bone1Name=dyn_skirt_back_r, bone2Name=pelvis, swing1LimitDegrees=45, swing2LimitDegrees=45, twistLimitDegrees=15}
 {bone1Name=dyn_skirt_back_l, bone2Name=pelvis, swing1LimitDegrees=45, swing2LimitDegrees=45, twistLimitDegrees=15}
 {
  bone1Name=thigh_l
  bone2Name=pelvis
  swing1LimitDegrees=45
  swing1Motion=Free
  swing2LimitDegrees=45
  swing2Motion=Free
  twistLimitDegrees=45
  twistMotion=Free
 }
 {
  bone1Name=thigh_r
  bone2Name=pelvis
  swing1LimitDegrees=45
  swing1Motion=Free
  swing2LimitDegrees=45
  swing2Motion=Free
  twistLimitDegrees=45
  twistMotion=Free
 }
 {
  bone1Name=dyn_skirt_back_r
  bone2Name=dyn_skirt_back_l
  swing1LimitDegrees=45
  swing1Motion=Free
  swing2LimitDegrees=45
  swing2Motion=Free
  twistLimitDegrees=45
  twistMotion=Free
 }
]
