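# Post-process-style AnimBP for the female Red Riding Remix outfit, adding secondary motion to the copied parent-mesh pose.
# RBF pose drivers set shoulder pads using separate grounded and airborne sets; back-jacket physics switches off when a backpack is equipped.
# Front skirt uses AnimDynamics, a crouch offset and thigh constraints, feeding RigidBody simulations with skydiving force.
# Right-arm lift drives both collar corrections, and updates reset dynamics outside water jumps.
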AnimBlueprint /BRCosmetics/Characters/Player/Female/Medium/Bodies/F_MED_RedRidingRemix/Meshes/F_MED_Red_Riding_Remix_T1_AnimBP
target_skeleton=&/F_MED_Red_Riding_Remix_T1_Skeleton
graphs [
 AnimGraph {nodes [
   n1 = "Constraint - dyn_skirt_front (TT)"!(ComponentPose=Local To Component(LocalPose=Use cached pose 'SkirtFix'()), Alpha=0.5, ConstraintWeights_0=1, ConstraintWeights_1:float=1) -> (Pose) {
    alphaBoolBlend {blendOption=Linear}
    boneToModify {boneName=dyn_skirt_front}
    constraintSetup [{transformType=Translation, targetBone {boneName=thigh_twist_01_l}}, {transformType=Translation, targetBone {boneName=thigh_twist_01_r}}]
    constraintWeights [1,1]
   }
   n2 = Break Fort Anim Input Skydiving External Force(FortAnimInput_SkydivingExternalForce=Get SkydivingExternalForceInput()) -> (
    bUseSkydivingVectorForce=~
    SkydivingVectorForce=~
    HeadToPelvisDirection=~
    FloatingMultiplier=~
    FloatingAdditive=~
    DivingMultiplier=~
    DivingAdditive=~
    ParachutingMultiplier=~
    ParachutingAdditive=~
   )
   n3 = RigidBody(ComponentPose=@n1, ExternalForce=@n2.SkydivingVectorForce) {
    bClampLinearTranslationLimitToRefPose=true
    bOverrideWorldGravity=true
    overridePhysicsAsset=&/F_MED_Red_Riding_Remix_T1_Skirt_Physics
    simulationSpace=BaseBoneSpace
    alphaBoolBlend {blendOption=Linear}
    baseBoneRef {boneName=root}
    componentAppliedLinearAccClamp=<6000 6000 6000>
    componentLinearAccScale=<0.3 0.3 0.3>
    overrideWorldGravity {z=-980}
   }
   n4 = RigidBody(ComponentPose=@n3, ExternalForce=@n2.SkydivingVectorForce) {
    bClampLinearTranslationLimitToRefPose=true
    bOverrideWorldGravity=true
    overridePhysicsAsset=&/F_MED_Red_Riding_Remix_T1_Physics
    simulationSpace=BaseBoneSpace
    alphaBoolBlend {blendOption=Linear}
    baseBoneRef {boneName=spine_05}
    componentAppliedLinearAccClamp=<6000 6000 6000>
    componentLinearAccScale=<0.4 0.4 0.4>
    overrideWorldGravity {z=-980}
   }
   n5 = RigidBody(ComponentPose=@n4, ExternalForce=@n2.SkydivingVectorForce) {
    bClampLinearTranslationLimitToRefPose=true
    bOverrideWorldGravity=true
    overridePhysicsAsset=&/F_MED_Red_Riding_Remix_T1_ShoeBuckle_Physics
    simulationSpace=BaseBoneSpace
    alphaBoolBlend {blendOption=Linear}
    baseBoneRef {boneName=root}
    componentAppliedLinearAccClamp=<6000 6000 6000>
    componentLinearAccScale=<0.2 0.2 0.2>
    overrideWorldGravity {z=-980}
   }
   n6 = Get ArmSideLiftR() -> (ArmSideLiftR:double)
   n7 = "Transform (Modify) Bone - Bone: collar_r"(ComponentPose=@n5, Alpha=@n6, Translation="0,0,0", Rotation="0,-10,0", Scale="1,1,1") {rotationMode=BMM_Additive, alphaBoolBlend {blendOption=Linear}, boneToModify {boneName=collar_r}, rotation {yaw=-10}}
   n8 = "Transform (Modify) Bone - Bone: collar_l"(ComponentPose=@n7, Alpha=@n6, Translation="-1,0,0", Rotation="0,0,-30", Scale="1,1,1") {
    rotationMode=BMM_Additive
    translationMode=BMM_Additive
    alphaBoolBlend {blendOption=Linear}
    boneToModify {boneName=collar_l}
    rotation {roll=-30}
    translation {x=-1}
   }
   Output Pose(Result=Component To Local(ComponentPose=@n8)) {layerGroup=None}
   Save cached pose 'BasePose'(Pose=Copy Pose From Mesh() {bUseAttachedParent=true}) {cacheName=BasePose}
   n9 = Pose Driver - Source: upperarm_l!(SourcePose:Pose Link Structure=Use cached pose 'BasePose'()) -> (Pose:Pose Link Structure) {
    is=title AnimGraphNode_PoseDriver
    poseAsset=&/Skydive_T1_PoseAsset
    onlyDriveBones [{boneName=shldr_pad_l}]
    poseTargets [
     {
      bApplyCustomCurve=false
      bIsHidden=false
      distanceMethod=DefaultMethod
      drivenName=Pose_13
      functionType=DefaultFunction
      targetScale=1
      boneTransforms [{targetRotation=<rot -37.9117 2.9184 0.9883>, targetTranslation=<9.8658 -1.2136 -0.4105>}]
      customCurve {defaultValue=3.4028234663852886e+38, postInfinityExtrap=RCCE_Constant, preInfinityExtrap=RCCE_Constant, keys []}
      targetRotation=<rot 0 0 0>
     }
     {
      bApplyCustomCurve=false
      bIsHidden=false
      distanceMethod=DefaultMethod
      drivenName=Pose_14
      functionType=DefaultFunction
      targetScale=0.4
      boneTransforms [{targetRotation=<rot -2.7179 3.7605 -70.8009>, targetTranslation=<9.8658 -1.2136 -0.4105>}]
      customCurve {defaultValue=3.4028234663852886e+38, postInfinityExtrap=RCCE_Constant, preInfinityExtrap=RCCE_Constant, keys []}
      targetRotation=<rot 0 0 0>
     }
     {
      bApplyCustomCurve=false
      bIsHidden=false
      distanceMethod=DefaultMethod
      drivenName=Pose_15
      functionType=DefaultFunction
      targetScale=0.4
      boneTransforms [{targetRotation=<rot 5.4251 8.6056 -70.3878>, targetTranslation=<9.8657 -1.2136 -0.4105>}]
      customCurve {defaultValue=3.4028234663852886e+38, postInfinityExtrap=RCCE_Constant, preInfinityExtrap=RCCE_Constant, keys []}
      targetRotation=<rot 0 0 0>
     }
     {
      bApplyCustomCurve=false
      bIsHidden=false
      distanceMethod=DefaultMethod
      drivenName=Pose_16
      functionType=DefaultFunction
      targetScale=0.6845
      boneTransforms [{targetRotation=<rot 42.756 59.0965 -59.0281>, targetTranslation=<9.8602 -1.2125 -0.4113>}]
      customCurve {defaultValue=3.4028234663852886e+38, postInfinityExtrap=RCCE_Constant, preInfinityExtrap=RCCE_Constant, keys []}
      targetRotation=<rot 0 0 0>
     }
     {
      bApplyCustomCurve=false
      bIsHidden=false
      distanceMethod=DefaultMethod
      drivenName=Pose_17
      functionType=DefaultFunction
      targetScale=0.9451
      boneTransforms [{targetRotation=<rot 9.4881 59.0505 -56.5461>, targetTranslation=<9.8657 -1.2135 -0.4105>}]
      customCurve {defaultValue=3.4028234663852886e+38, postInfinityExtrap=RCCE_Constant, preInfinityExtrap=RCCE_Constant, keys []}
      targetRotation=<rot 0 0 0>
     }
     {
      bApplyCustomCurve=false
      bIsHidden=false
      distanceMethod=DefaultMethod
      drivenName=Pose_18
      functionType=DefaultFunction
      targetScale=0.6949
      boneTransforms [{targetRotation=<rot 30.8452 -1.6689 -66.9927>, targetTranslation=<9.8658 -1.2136 -0.4104>}]
      customCurve {defaultValue=3.4028234663852886e+38, postInfinityExtrap=RCCE_Constant, preInfinityExtrap=RCCE_Constant, keys []}
      targetRotation=<rot 0 0 0>
     }
     {
      bApplyCustomCurve=false
      bIsHidden=false
      distanceMethod=DefaultMethod
      drivenName=Pose_19
      functionType=DefaultFunction
      targetScale=0.6845
      boneTransforms [{targetRotation=<rot 38.9497 27.4481 -68.1025>, targetTranslation=<9.8657 -1.2136 -0.4105>}]
      customCurve {defaultValue=3.4028234663852886e+38, postInfinityExtrap=RCCE_Constant, preInfinityExtrap=RCCE_Constant, keys []}
      targetRotation=<rot 0 0 0>
     }
     {
      bApplyCustomCurve=false
      bIsHidden=false
      distanceMethod=DefaultMethod
      drivenName=Pose_20
      functionType=DefaultFunction
      targetScale=0.4521
      boneTransforms [{targetRotation=<rot 7.2733 -8.6534 -82.7219>, targetTranslation=<9.865 -1.2137 -0.4104>}]
      customCurve {defaultValue=3.4028234663852886e+38, postInfinityExtrap=RCCE_Constant, preInfinityExtrap=RCCE_Constant, keys []}
      targetRotation=<rot 0 0 0>
     }
     {
      bApplyCustomCurve=false
      bIsHidden=false
      distanceMethod=DefaultMethod
      drivenName=Pose_21
      functionType=DefaultFunction
      targetScale=0.5196
      boneTransforms [{targetRotation=<rot -4.6609 -22.5417 -85.6617>, targetTranslation=<9.8652 -1.2133 -0.4105>}]
      customCurve {defaultValue=3.4028234663852886e+38, postInfinityExtrap=RCCE_Constant, preInfinityExtrap=RCCE_Constant, keys []}
      targetRotation=<rot 0 0 0>
     }
    ]
    rBFParams {radius=30}
    sourceBones [{boneName=upperarm_l}]
   }
   n10 = Pose Driver - Source: upperarm_r!(SourcePose:Pose Link Structure=@n9) -> (Pose:Pose Link Structure) {
    is=title AnimGraphNode_PoseDriver
    poseAsset=&Skydive_T1_PoseAsset
    onlyDriveBones [{boneName=shldr_pad_r}]
    poseTargets [
     {
      bApplyCustomCurve=false
      bIsHidden=false
      distanceMethod=DefaultMethod
      drivenName=Pose_13
      functionType=DefaultFunction
      targetScale=1
      boneTransforms [{targetRotation=<rot -37.9117 2.9184 0.9883>, targetTranslation=<-9.8658 1.2136 0.4104>}]
      customCurve {defaultValue=3.4028234663852886e+38, postInfinityExtrap=RCCE_Constant, preInfinityExtrap=RCCE_Constant, keys []}
      targetRotation=<rot 0 0 0>
     }
     {
      bApplyCustomCurve=false
      bIsHidden=false
      distanceMethod=DefaultMethod
      drivenName=Pose_14
      functionType=DefaultFunction
      targetScale=0.0064
      boneTransforms [{targetRotation=<rot -3.6429 0.34 -58.8815>, targetTranslation=<-9.8657 1.2136 0.4105>}]
      customCurve {defaultValue=3.4028234663852886e+38, postInfinityExtrap=RCCE_Constant, preInfinityExtrap=RCCE_Constant, keys []}
      targetRotation=<rot 0 0 0>
     }
     {
      bApplyCustomCurve=false
      bIsHidden=false
      distanceMethod=DefaultMethod
      drivenName=Pose_15
      functionType=DefaultFunction
      targetScale=0.6
      boneTransforms [{targetRotation=<rot -3.8274 0.2226 -58.9375>, targetTranslation=<-9.8656 1.2136 0.4105>}]
      customCurve {defaultValue=3.4028234663852886e+38, postInfinityExtrap=RCCE_Constant, preInfinityExtrap=RCCE_Constant, keys []}
      targetRotation=<rot 0 0 0>
     }
     {
      bApplyCustomCurve=false
      bIsHidden=false
      distanceMethod=DefaultMethod
      drivenName=Pose_16
      functionType=DefaultFunction
      targetScale=0.938
      boneTransforms [{targetRotation=<rot -9.2756 32.1153 77.4468>, targetTranslation=<-9.8599 1.2129 0.409>}]
      customCurve {defaultValue=3.4028234663852886e+38, postInfinityExtrap=RCCE_Constant, preInfinityExtrap=RCCE_Constant, keys []}
      targetRotation=<rot 0 0 0>
     }
     {
      bApplyCustomCurve=false
      bIsHidden=false
      distanceMethod=DefaultMethod
      drivenName=Pose_17
      functionType=DefaultFunction
      targetScale=0.4992
      boneTransforms [{targetRotation=<rot 35.1238 13.4854 -111.8821>, targetTranslation=<-9.8605 1.2131 0.4099>}]
      customCurve {defaultValue=3.4028234663852886e+38, postInfinityExtrap=RCCE_Constant, preInfinityExtrap=RCCE_Constant, keys []}
      targetRotation=<rot 0 0 0>
     }
     {
      bApplyCustomCurve=false
      bIsHidden=false
      distanceMethod=DefaultMethod
      drivenName=Pose_18
      functionType=DefaultFunction
      targetScale=1
      boneTransforms [{targetRotation=<rot 23.3238 -0.6909 -73.8606>, targetTranslation=<-9.8657 1.2136 0.4105>}]
      customCurve {defaultValue=3.4028234663852886e+38, postInfinityExtrap=RCCE_Constant, preInfinityExtrap=RCCE_Constant, keys []}
      targetRotation=<rot 0 0 0>
     }
     {
      bApplyCustomCurve=false
      bIsHidden=false
      distanceMethod=DefaultMethod
      drivenName=Pose_19
      functionType=DefaultFunction
      targetScale=0.5
      boneTransforms [{targetRotation=<rot 23.3238 -0.6909 -73.8606>, targetTranslation=<-9.8657 1.2136 0.4105>}]
      customCurve {defaultValue=3.4028234663852886e+38, postInfinityExtrap=RCCE_Constant, preInfinityExtrap=RCCE_Constant, keys []}
      targetRotation=<rot 0 0 0>
     }
     {
      bApplyCustomCurve=false
      bIsHidden=false
      distanceMethod=DefaultMethod
      drivenName=Pose_20
      functionType=DefaultFunction
      targetScale=0.4464
      boneTransforms [{targetRotation=<rot 12.787 -23.851 -88.4562>, targetTranslation=<-9.8656 1.2135 0.4106>}]
      customCurve {defaultValue=3.4028234663852886e+38, postInfinityExtrap=RCCE_Constant, preInfinityExtrap=RCCE_Constant, keys []}
      targetRotation=<rot 0 0 0>
     }
     {
      bApplyCustomCurve=false
      bIsHidden=false
      distanceMethod=DefaultMethod
      drivenName=Pose_21
      functionType=DefaultFunction
      targetScale=0.2788
      boneTransforms [{targetRotation=<rot 21.6748 -10.8512 -81.4257>, targetTranslation=<-9.8635 1.2137 0.4104>}]
      customCurve {defaultValue=3.4028234663852886e+38, postInfinityExtrap=RCCE_Constant, preInfinityExtrap=RCCE_Constant, keys []}
      targetRotation=<rot 0 0 0>
     }
    ]
    rBFParams {radius=30}
    sourceBones [{boneName=upperarm_r}]
   }
   n11 = Pose Driver - Source: upperarm_l(SourcePose=Use cached pose 'BasePose'()) {
    poseAsset=&/F_MED_Red_Riding_Remix_ShldrT1_RBF__PoseAsset
    onlyDriveBones [{boneName=shldr_pad_l}]
    poseTargets [
     {
      bApplyCustomCurve=false
      bIsHidden=false
      distanceMethod=DefaultMethod
      drivenName=Base
      functionType=DefaultFunction
      targetScale=1
      boneTransforms [
       {targetRotation=<rot -37.9117 2.9184 0.9883>, targetTranslation=<9.8658 -1.2136 -0.4105>}
       {targetRotation=<rot -85.1913 -138.5624 143.0618>, targetTranslation=<5.483 -0.1053 -3.2946>}
      ]
      customCurve {defaultValue=3.4028234663852886e+38, postInfinityExtrap=RCCE_Constant, preInfinityExtrap=RCCE_Constant, keys []}
      targetRotation=<rot 0 0 0>
     }
     {
      bApplyCustomCurve=false
      bIsHidden=false
      distanceMethod=DefaultMethod
      drivenName=ADS_Forward
      functionType=DefaultFunction
      targetScale=0.3211
      boneTransforms [
       {targetRotation=<rot -9.8194 59.4736 -38.2892>, targetTranslation=<9.6835 -1.1702 -0.3611>}
       {targetRotation=<rot -71.3616 49.7635 -58.8898>, targetTranslation=<5.5311 -0.2285 -3.2544>}
      ]
      customCurve {defaultValue=3.4028234663852886e+38, postInfinityExtrap=RCCE_Constant, preInfinityExtrap=RCCE_Constant, keys []}
      targetRotation=<rot 0 0 0>
     }
     {
      bApplyCustomCurve=false
      bIsHidden=false
      distanceMethod=DefaultMethod
      drivenName=ADS_Down
      functionType=DefaultFunction
      targetScale=1
      boneTransforms [
       {targetRotation=<rot -11.0898 -10.2259 -31.2827>, targetTranslation=<9.8655 -1.2152 -0.4114>}
       {targetRotation=<rot -72.0525 -2.9517 -34.351>, targetTranslation=<7.6797 -4.2937 -2.4568>}
      ]
      customCurve {defaultValue=3.4028234663852886e+38, postInfinityExtrap=RCCE_Constant, preInfinityExtrap=RCCE_Constant, keys []}
      targetRotation=<rot 0 0 0>
     }
     {
      bApplyCustomCurve=false
      bIsHidden=false
      distanceMethod=DefaultMethod
      drivenName=ADS_Up
      functionType=DefaultFunction
      targetScale=0.7567
      boneTransforms [
       {targetRotation=<rot -17.1538 82.6063 -12.0746>, targetTranslation=<9.8634 -1.2129 -0.4095>}
       {targetRotation=<rot -79.0327 99.4722 -111.5475>, targetTranslation=<6.2052 -1.4819 -3.0211>}
      ]
      customCurve {defaultValue=3.4028234663852886e+38, postInfinityExtrap=RCCE_Constant, preInfinityExtrap=RCCE_Constant, keys []}
      targetRotation=<rot 0 0 0>
     }
     {
      bApplyCustomCurve=false
      bIsHidden=false
      distanceMethod=DefaultMethod
      drivenName=Pose_8
      functionType=DefaultFunction
      targetScale=0.7042
      boneTransforms [
       {targetRotation=<rot -0.5452 60.5478 -101.8424>, targetTranslation=<9.8666 -1.2134 -0.4092>}
       {targetRotation=<rot -53.7698 7.5873 -4.924>, targetTranslation=<5.4836 -0.1051 -3.2947>}
      ]
      customCurve {defaultValue=3.4028234663852886e+38, postInfinityExtrap=RCCE_Constant, preInfinityExtrap=RCCE_Constant, keys []}
      targetRotation=<rot 0 0 0>
     }
     {
      bApplyCustomCurve=false
      bIsHidden=false
      distanceMethod=DefaultMethod
      drivenName=Pose_9
      functionType=DefaultFunction
      targetScale=0.6382
      boneTransforms [
       {targetRotation=<rot -73.5602 73.364 -47.481>, targetTranslation=<9.8658 -1.2136 -0.4096>}
       {targetRotation=<rot -82.307 65.6231 -61.9678>, targetTranslation=<5.4829 -0.1054 -3.2946>}
      ]
      customCurve {defaultValue=3.4028234663852886e+38, postInfinityExtrap=RCCE_Constant, preInfinityExtrap=RCCE_Constant, keys []}
      targetRotation=<rot 0 0 0>
     }
     {
      bApplyCustomCurve=false
      bIsHidden=false
      distanceMethod=DefaultMethod
      drivenName=Pose_10
      functionType=DefaultFunction
      targetScale=0.7042
      boneTransforms [
       {targetRotation=<rot -12.8643 37.8653 -76.7391>, targetTranslation=<9.8682 -1.2132 -0.4082>}
       {targetRotation=<rot -54.5122 7.7596 -5.3834>, targetTranslation=<5.484 -0.105 -3.2949>}
      ]
      customCurve {defaultValue=3.4028234663852886e+38, postInfinityExtrap=RCCE_Constant, preInfinityExtrap=RCCE_Constant, keys []}
      targetRotation=<rot 0 0 0>
     }
     {
      bApplyCustomCurve=false
      bIsHidden=false
      distanceMethod=DefaultMethod
      drivenName=Pose_11
      functionType=DefaultFunction
      targetScale=0.6382
      boneTransforms [
       {targetRotation=<rot -52.3047 60.2615 -31.9336>, targetTranslation=<9.8687 -1.2074 -0.4149>}
       {targetRotation=<rot -87.4449 144.5891 -140.0223>, targetTranslation=<5.4775 -0.1035 -3.2957>}
      ]
      customCurve {defaultValue=3.4028234663852886e+38, postInfinityExtrap=RCCE_Constant, preInfinityExtrap=RCCE_Constant, keys []}
      targetRotation=<rot 0 0 0>
     }
     {
      bApplyCustomCurve=false
      bIsHidden=false
      distanceMethod=DefaultMethod
      drivenName=Pose_12
      functionType=DefaultFunction
      targetScale=0.3211
      boneTransforms [
       {targetRotation=<rot -8.3931 58.8074 -53.7213>, targetTranslation=<9.8656 -1.2136 -0.4105>}
       {targetRotation=<rot -59.8269 52.5616 -56.4917>, targetTranslation=<5.4828 -0.1052 -3.2946>}
      ]
      customCurve {defaultValue=3.4028234663852886e+38, postInfinityExtrap=RCCE_Constant, preInfinityExtrap=RCCE_Constant, keys []}
      targetRotation=<rot 0 0 0>
     }
    ]
    rBFParams {radius=35}
    sourceBones [{boneName=upperarm_l}, {boneName=clavicle_l}]
   }
   n12 = Pose Driver - Source: upperarm_r(SourcePose=@n11) {
    poseAsset=&F_MED_Red_Riding_Remix_ShldrT1_RBF__PoseAsset
    onlyDriveBones [{boneName=shldr_pad_r}]
    poseTargets [
     {
      bApplyCustomCurve=false
      bIsHidden=false
      distanceMethod=DefaultMethod
      drivenName=Base
      functionType=DefaultFunction
      targetScale=0.7201
      boneTransforms [{targetRotation=<rot -37.9117 2.9184 0.9883>, targetTranslation=<-9.8658 1.2136 0.4104>}]
      customCurve {defaultValue=3.4028234663852886e+38, postInfinityExtrap=RCCE_Constant, preInfinityExtrap=RCCE_Constant, keys []}
      targetRotation=<rot 0 0 0>
     }
     {
      bApplyCustomCurve=false
      bIsHidden=false
      distanceMethod=DefaultMethod
      drivenName=ADS_Forward
      functionType=DefaultFunction
      targetScale=0.8
      boneTransforms [{targetRotation=<rot -12.5256 49.7852 -85.7463>, targetTranslation=<-9.8128 1.2251 0.4087>}]
      customCurve {defaultValue=3.4028234663852886e+38, postInfinityExtrap=RCCE_Constant, preInfinityExtrap=RCCE_Constant, keys []}
      targetRotation=<rot 0 0 0>
     }
     {
      bApplyCustomCurve=false
      bIsHidden=false
      distanceMethod=DefaultMethod
      drivenName=ADS_Down
      functionType=DefaultFunction
      targetScale=1
      boneTransforms [{targetRotation=<rot -21.6145 102.2849 -29.0379>, targetTranslation=<-9.8664 1.2124 0.4093>}]
      customCurve {defaultValue=3.4028234663852886e+38, postInfinityExtrap=RCCE_Constant, preInfinityExtrap=RCCE_Constant, keys []}
      targetRotation=<rot 0 0 0>
     }
     {
      bApplyCustomCurve=false
      bIsHidden=false
      distanceMethod=DefaultMethod
      drivenName=ADS_Up
      functionType=DefaultFunction
      targetScale=0.9
      boneTransforms [{targetRotation=<rot -11.8597 31.0478 -74.766>, targetTranslation=<-9.8648 1.2138 0.4096>}]
      customCurve {defaultValue=3.4028234663852886e+38, postInfinityExtrap=RCCE_Constant, preInfinityExtrap=RCCE_Constant, keys []}
      targetRotation=<rot 0 0 0>
     }
     {
      bApplyCustomCurve=false
      bIsHidden=false
      distanceMethod=DefaultMethod
      drivenName=Pose_8
      functionType=DefaultFunction
      targetScale=1
      boneTransforms [{targetRotation=<rot 28.0696 94.5729 -83.9531>, targetTranslation=<-9.8648 1.2138 0.4129>}]
      customCurve {defaultValue=3.4028234663852886e+38, postInfinityExtrap=RCCE_Constant, preInfinityExtrap=RCCE_Constant, keys []}
      targetRotation=<rot 0 0 0>
     }
     {
      bApplyCustomCurve=false
      bIsHidden=false
      distanceMethod=DefaultMethod
      drivenName=Pose_9
      functionType=DefaultFunction
      targetScale=0.5
      boneTransforms [{targetRotation=<rot -46.0602 112.0651 -51.2415>, targetTranslation=<-9.8654 1.2136 0.4113>}]
      customCurve {defaultValue=3.4028234663852886e+38, postInfinityExtrap=RCCE_Constant, preInfinityExtrap=RCCE_Constant, keys []}
      targetRotation=<rot 0 0 0>
     }
     {
      bApplyCustomCurve=false
      bIsHidden=false
      distanceMethod=DefaultMethod
      drivenName=Pose_10
      functionType=DefaultFunction
      targetScale=0.4
      boneTransforms [{targetRotation=<rot -35.6283 113.0032 -69.6103>, targetTranslation=<-9.8653 1.2142 0.4154>}]
      customCurve {defaultValue=3.4028234663852886e+38, postInfinityExtrap=RCCE_Constant, preInfinityExtrap=RCCE_Constant, keys []}
      targetRotation=<rot 0 0 0>
     }
     {
      bApplyCustomCurve=false
      bIsHidden=false
      distanceMethod=DefaultMethod
      drivenName=Pose_11
      functionType=DefaultFunction
      targetScale=0.4272
      boneTransforms [{targetRotation=<rot -66.4497 124.6636 -76.7572>, targetTranslation=<-9.87 1.2195 0.4082>}]
      customCurve {defaultValue=3.4028234663852886e+38, postInfinityExtrap=RCCE_Constant, preInfinityExtrap=RCCE_Constant, keys []}
      targetRotation=<rot 0 0 0>
     }
     {
      bApplyCustomCurve=false
      bIsHidden=false
      distanceMethod=DefaultMethod
      drivenName=Pose_12
      functionType=DefaultFunction
      targetScale=0.5
      boneTransforms [{targetRotation=<rot -14.5595 55.6884 -76.4529>, targetTranslation=<-9.8654 1.2134 0.4104>}]
      customCurve {defaultValue=3.4028234663852886e+38, postInfinityExtrap=RCCE_Constant, preInfinityExtrap=RCCE_Constant, keys []}
      targetRotation=<rot 0 0 0>
     }
    ]
    rBFParams {radius=35}
    sourceBones [{boneName=upperarm_r}]
   }
   Save cached pose 'postPoseDriver'(Pose=Blend Poses by bool(BlendPose_0=@n12, BlendPose_1=@n10, bActiveValue=Get bIsOnGround() -> (bIsOnGround:bool)) {blendType=Linear}) {cacheName=postPoseDriver}
   n13 = RigidBody(ComponentPose=Local To Component(LocalPose=Use cached pose 'postPoseDriver'()), Alpha=0) {
    bClampLinearTranslationLimitToRefPose=true
    bOverrideWorldGravity=true
    overridePhysicsAsset=&/F_MED_Red_Riding_Remix_BackJacket_Physics
    simulationSpace=BaseBoneSpace
    alphaBoolBlend {blendOption=Linear}
    baseBoneRef {boneName=spine_05}
    componentAppliedLinearAccClamp=<6000 6000 6000>
    componentLinearAccScale=<0.4 0.4 0.4>
    overrideWorldGravity {z=-980}
   }
   n14 = RigidBody(ComponentPose=Local To Component(LocalPose=Use cached pose 'postPoseDriver'()), ExternalForce=@n2.SkydivingVectorForce) {
    bClampLinearTranslationLimitToRefPose=true
    bOverrideWorldGravity=true
    overridePhysicsAsset=&F_MED_Red_Riding_Remix_BackJacket_Physics
    simulationSpace=BaseBoneSpace
    alphaBoolBlend {blendOption=Linear}
    baseBoneRef {boneName=spine_05}
    componentAppliedLinearAccClamp=<6000 6000 6000>
    componentLinearAccScale=<0.4 0.4 0.4>
    overrideWorldGravity {z=-980}
   }
   Save cached pose 'BackpackDynFix'(Pose=Blend Poses by bool(
     BlendPose_0=Component To Local(ComponentPose=@n13)
     BlendPose_1=Component To Local(ComponentPose=@n14)
     bActiveValue=Get IsBackpackEquipped() -> (IsBackpackEquipped:bool)
    ) {blendType=Linear}) {cacheName=BackpackDynFix}
   n15 = Anim Dynamics - Bone: dyn_skirt_front(ComponentPose=Local To Component(LocalPose=Use cached pose 'BackpackDynFix'()), Alpha=1, ExternalForce=@n2.SkydivingVectorForce) {
    angularDampingOverride=0.98
    bGravityOverrideInSimSpace=true
    bOverrideAngularDamping=true
    bUseGravityOverride=true
    simulationSpace=RootRelative
    alphaBoolBlend {blendOption=Linear}
    boundBone {boneName=dyn_skirt_front}
    gravityOverride {z=-980}
    physicsBodyDefinitions [{
      sphereCollisionRadius=0
      boundBone {boneName=dyn_skirt_front}
      boxExtents=<6 1 6>
      constraintSetup {angularLimitsMax=<0 0 0>, angularLimitsMin=<0 0 -20>, angularTarget=<0 0 0>}
      localJointOffset=<3 0 0>
     }]
   }
   n16 = Anim Dynamics - Bone: dyn_skirt_front(ComponentPose=Local To Component(LocalPose=Use cached pose 'BackpackDynFix'()), Alpha=1, ExternalForce="0,0,0") {
    angularDampingOverride=0.98
    bGravityOverrideInSimSpace=true
    bOverrideAngularDamping=true
    bUseGravityOverride=true
    simulationSpace=RootRelative
    alphaBoolBlend {blendOption=Linear}
    boundBone {boneName=dyn_skirt_front}
    gravityOverride {z=-980}
    physicsBodyDefinitions [{
      sphereCollisionRadius=0
      boundBone {boneName=dyn_skirt_front}
      boxExtents=<6 1 6>
      constraintSetup {angularLimitsMax=<0 0 0>, angularLimitsMin=<0 0 -20>, angularTarget=<0 0 0>}
      localJointOffset=<3 0 0>
     }]
   }
   n17 = "Transform (Modify) Bone - Bone: dyn_skirt_front"(ComponentPose=@n16, Alpha=1, Translation="2,0,7", Rotation="0,0,0", Scale="1,1,1") {translationMode=BMM_Additive, alphaBoolBlend {blendOption=Linear}, boneToModify {boneName=dyn_skirt_front}, translation {x=2, z=7}}
   Save cached pose 'SkirtFix'(Pose=Blend Poses by bool(BlendPose_0=Component To Local(ComponentPose=@n17), BlendPose_1=Component To Local(ComponentPose=@n15), bActiveValue=Get bIsCrouching()) {blendType=Linear}) {cacheName=SkirtFix}
  ]}
 EventGraph {nodes [
   Event OnCharacterCustomizationLoaded()
   n1 = Get PlayerPawn() -> (PlayerPawn:Fort Player Pawn Athena Object Reference)
   > Is Valid(InputObject=@n1)
   > Do Once()
   n2 = CheckForBackbling()
   > Set IsBackpackEquipped(IsBackpackEquipped:bool=OR Boolean(A=@n2, B=@n2.bIsBackblingLong)) -> (Output_Get:bool)
   Event BlueprintInitializeAnimation()
   n3 = TryGetPawnOwner()
   > Is Valid(InputObject=@n3)
   > n4 = Cast To FortPlayerPawnAthena(Object=@n3)
   > Set PlayerPawn(PlayerPawn:Fort Player Pawn Athena Object Reference=@n4) -> (Output_Get:Fort Player Pawn Athena Object Reference)
   Event BlueprintUpdateAnimation()
   > n5 = Sequence!(execute) -> (then_0, then_1, then_2:Exec)
   n6 = GetDeltaTransformFromRefPose(self=GetOwnerSkeletalMeshComponent(), BoneName=upperarm_r, BaseName=clavicle_r)
   n7 = "Absolute (Float)"(A=BreakRotator(InRot=BreakTransform(InTransform=@n6).Rotation).Yaw)
   Set ArmSideLiftR(execute=@n5.then_1, ArmSideLiftR:double=float * float(A=MapRangeClamped(Value=@n7, InRangeA=40, InRangeB=120), B=1.5)) -> (Output_Get:double)
   n8 = Branch(execute=@n5.then_2, Condition=Get bIsWaterJump() -> (bIsWaterJump:bool))
   ResetDynamics(execute=@n8.else, InTeleportType=TeleportPhysics)
   > ResetDynamics(InTeleportType=ResetPhysics)
   "Not Equal (Object)"!(A:Object Reference=GetSkeletalMeshForPartType(self=@n1, PartType=Backpack), B:Object Reference) -> (ReturnValue:bool)
   TryGetPawnOwner()
  ]}
]
variables [
 ArmSideLiftR {type=number}
 IsBackpackEquipped {type=boolean}
 PlayerPawn {type=FortPlayerPawnAthena}
]
