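# Physics asset for the shoulder pads of the female medium Red Riding Remix player cosmetic, supporting physics on the pads' fur pieces.
# A pelvis capsule and a box on each shoulder pad form the base, with free swing and twist between pelvis and pads.
# Back and bottom fur boxes on each side constrain to their own pad, with a tighter twist limit than swing.
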PhysicsAsset /BRCosmetics/Characters/Player/Female/Medium/Bodies/F_MED_RedRidingRemix/Meshes/F_MED_Red_Riding_Remix_T2ShldrPads_Physics
bodies {
 dyn_shldr_fur_back_l [{shapeType=Box, extentX=3.1167, extentY=8.2471, extentZ=4.3294, center=<0 0 -1.1842>}]
 dyn_shldr_fur_back_r [{shapeType=Box, extentX=3.6167, extentY=8.2471, extentZ=4.3294, center=<0.008 -0.1581 1.1735>}]
 dyn_shldr_fur_btm_l [{shapeType=Box, extentX=8.1167, extentY=8.2471, extentZ=4.3294, center=<1 0 0>}]
 dyn_shldr_fur_btm_r [{shapeType=Box, extentX=8.1167, extentY=8.2471, extentZ=4.3294, center=<-1 0 0>}]
 pelvis [{shapeType=Capsule, length=6.0559, radius=8.7569, center=<-11.6519 -1.4623 -0.0004>, rotation=<rot 90 0 0>}]
 shldr_pad_l [{shapeType=Box, extentX=15.4246, extentY=10.8013, extentZ=6.876, center=<5.2176 0.0725 -2.3172>, rotation=<rot -21.7402 0 0>}]
 shldr_pad_r [{shapeType=Box, extentX=15.4246, extentY=10.8013, extentZ=6.876, center=<-5.2176 -0.0725 2.3169>, rotation=<rot -21.7403 0 0>}]
}
constraints [
 {
  bone1Name=shldr_pad_r
  bone2Name=pelvis
  swing1LimitDegrees=45
  swing1Motion=Free
  swing2LimitDegrees=45
  swing2Motion=Free
  twistLimitDegrees=45
  twistMotion=Free
 }
 {bone1Name=dyn_shldr_fur_btm_r, bone2Name=shldr_pad_r, swing1LimitDegrees=45, swing2LimitDegrees=45, twistLimitDegrees=15}
 {
  bone1Name=shldr_pad_l
  bone2Name=pelvis
  swing1LimitDegrees=45
  swing1Motion=Free
  swing2LimitDegrees=45
  swing2Motion=Free
  twistLimitDegrees=45
  twistMotion=Free
 }
 {bone1Name=dyn_shldr_fur_btm_l, bone2Name=shldr_pad_l, swing1LimitDegrees=45, swing2LimitDegrees=45, twistLimitDegrees=15}
 {bone1Name=dyn_shldr_fur_back_r, bone2Name=shldr_pad_r, swing1LimitDegrees=45, swing2LimitDegrees=45, twistLimitDegrees=15}
 {bone1Name=dyn_shldr_fur_back_l, bone2Name=shldr_pad_l, swing1LimitDegrees=45, swing2LimitDegrees=45, twistLimitDegrees=15}
]
