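# Physics asset for the female medium Red Riding Remix T2 player cosmetic body, set up for secondary motion on its buckles and back skirt.
# Capsules cover the pelvis, thighs and calf twist bones, while boxes cover four dyn_buckle bones per leg and two back skirt bones.
# Buckles are constrained to the calf twist bones and skirt pieces to the pelvis with swing and twist limit angles.
# Leg-chain constraints leave angular motion free.
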PhysicsAsset /BRCosmetics/Characters/Player/Female/Medium/Bodies/F_MED_RedRidingRemix/Meshes/F_MED_Red_Riding_Remix_T2_Physics
bodies {
 calf_twist_01_l [{shapeType=Capsule, length=4.9568, radius=4.6502, center=<-1.0521 -0.7886 -0.1189>, rotation=<rot 90 0 0>}]
 calf_twist_01_r [{shapeType=Capsule, length=4.9568, radius=4.6502, center=<1.0519 0.7886 0.1189>, rotation=<rot 90 0 0>}]
 calf_twist_02_l [{shapeType=Capsule, length=9.8346, radius=6.0402, center=<-3.9424 -2.3647 -0.4053>, rotation=<rot 90 0 0>}]
 calf_twist_02_r [{shapeType=Capsule, length=9.8346, radius=6.0402, center=<3.9423 2.3647 0.4052>, rotation=<rot 90 0 0>}]
 dyn_buckle_l_01 [{shapeType=Box, extentX=4, extentY=4, extentZ=1, center=<0 -1 0>}]
 dyn_buckle_l_02 [{shapeType=Box, extentX=4, extentY=4, extentZ=1, center=<0 -1 0>}]
 dyn_buckle_l_03 [{shapeType=Box, extentX=4, extentY=4, extentZ=1, center=<0 -1 0>}]
 dyn_buckle_l_04 [{shapeType=Box, extentX=4, extentY=4, extentZ=1, center=<0 -1 0>}]
 dyn_buckle_r_01 [{shapeType=Box, extentX=4, extentY=4, extentZ=1, center=<0 1 0>}]
 dyn_buckle_r_02 [{shapeType=Box, extentX=4, extentY=4, extentZ=1, center=<0 1 0>}]
 dyn_buckle_r_03 [{shapeType=Box, extentX=4, extentY=4, extentZ=1, center=<0 1 0>}]
 dyn_buckle_r_04 [{shapeType=Box, extentX=4, extentY=4, extentZ=1, center=<0 1 0>}]
 dyn_skirt_back_l [{shapeType=Box, extentX=15, extentY=3, extentZ=20, center=<7.157 1.3956 0>}]
 dyn_skirt_back_r [{shapeType=Box, extentX=15, extentY=3, extentZ=17, center=<7 1.3957 0>}]
 pelvis [{shapeType=Capsule, length=8.0559, radius=6.7569, center=<-11.6519 -1.4623 -0.0004>, rotation=<rot 90 0 0>}]
 thigh_l [{shapeType=Capsule, length=15.8649, radius=8.2416, center=<-21.4786 0.444 -0.8776>, rotation=<rot 90 0 0>}]
 thigh_r [{shapeType=Capsule, length=15.8649, radius=8.2412, center=<21.811 -0.3783 0.8772>, rotation=<rot 90 0 0>}]
}
constraints [
 {
  bone1Name=calf_twist_01_r
  bone2Name=thigh_r
  swing1LimitDegrees=45
  swing1Motion=Free
  swing2LimitDegrees=45
  swing2Motion=Free
  twistLimitDegrees=45
  twistMotion=Free
 }
 {bone1Name=dyn_buckle_r_03, bone2Name=calf_twist_01_r, swing1LimitDegrees=45, swing2LimitDegrees=45, twistLimitDegrees=15}
 {bone1Name=dyn_buckle_r_04, bone2Name=calf_twist_01_r, swing1LimitDegrees=45, swing2LimitDegrees=45, twistLimitDegrees=15}
 {
  bone1Name=calf_twist_02_r
  bone2Name=thigh_r
  swing1LimitDegrees=45
  swing1Motion=Free
  swing2LimitDegrees=45
  swing2Motion=Free
  twistLimitDegrees=45
  twistMotion=Free
 }
 {bone1Name=dyn_buckle_r_01, bone2Name=calf_twist_02_r, swing1LimitDegrees=45, swing2LimitDegrees=45, twistLimitDegrees=15}
 {bone1Name=dyn_buckle_r_02, bone2Name=calf_twist_02_r, swing1LimitDegrees=45, swing2LimitDegrees=45, twistLimitDegrees=15}
 {
  bone1Name=calf_twist_01_l
  bone2Name=thigh_l
  swing1LimitDegrees=45
  swing1Motion=Free
  swing2LimitDegrees=45
  swing2Motion=Free
  twistLimitDegrees=45
  twistMotion=Free
 }
 {bone1Name=dyn_buckle_l_04, bone2Name=calf_twist_01_l, swing1LimitDegrees=45, swing2LimitDegrees=45, twistLimitDegrees=15}
 {bone1Name=dyn_buckle_l_03, bone2Name=calf_twist_01_l, swing1LimitDegrees=45, swing2LimitDegrees=45, twistLimitDegrees=15}
 {
  bone1Name=calf_twist_02_l
  bone2Name=thigh_l
  swing1LimitDegrees=45
  swing1Motion=Free
  swing2LimitDegrees=45
  swing2Motion=Free
  twistLimitDegrees=45
  twistMotion=Free
 }
 {bone1Name=dyn_buckle_l_02, bone2Name=calf_twist_02_l, swing1LimitDegrees=45, swing2LimitDegrees=45, twistLimitDegrees=15}
 {bone1Name=dyn_buckle_l_01, bone2Name=calf_twist_02_l, swing1LimitDegrees=45, swing2LimitDegrees=45, twistLimitDegrees=15}
 {bone1Name=dyn_skirt_back_r, bone2Name=pelvis, swing1LimitDegrees=45, swing2LimitDegrees=45, twistLimitDegrees=15}
 {
  bone1Name=thigh_l
  bone2Name=pelvis
  swing1LimitDegrees=45
  swing1Motion=Free
  swing2LimitDegrees=45
  swing2Motion=Free
  twistLimitDegrees=45
  twistMotion=Free
 }
 {
  bone1Name=thigh_r
  bone2Name=pelvis
  swing1LimitDegrees=45
  swing1Motion=Free
  swing2LimitDegrees=45
  swing2Motion=Free
  twistLimitDegrees=45
  twistMotion=Free
 }
 {bone1Name=dyn_skirt_back_l, bone2Name=pelvis, swing1LimitDegrees=45, swing2LimitDegrees=45, twistLimitDegrees=15}
]
