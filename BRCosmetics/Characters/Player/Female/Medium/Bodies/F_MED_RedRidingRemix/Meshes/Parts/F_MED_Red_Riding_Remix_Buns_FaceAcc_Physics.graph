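# Physics asset for the Red Riding Remix female character's buns face accessory, defining secondary-motion physics for two hair buns, hair strands and a headband.
# Capsules cover buns and hair, boxes cover headband segments, and larger head and right clavicle capsules provide body shapes.
# Limited swing and twist constraints chain segments to the head, with right-mid and left-back hair branching off other strands.
# The right clavicle has no constraint.
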PhysicsAsset /BRCosmetics/Characters/Player/Female/Medium/Bodies/F_MED_RedRidingRemix/Meshes/Parts/F_MED_Red_Riding_Remix_Buns_FaceAcc_Physics
bodies {
 clavicle_r [{shapeType=Capsule, length=8.376, radius=7.5735, center=<-15.8758 -6.9698 -10.0156>, rotation=<rot -0.5467 14.7739 92.0729>}]
 dyn_bun_l [{shapeType=Capsule, length=1, radius=5, center=<1 0 0>, rotation=<rot 90 0 0>}]
 dyn_bun_r [{shapeType=Capsule, length=1, radius=5, center=<1.2065 0 0>, rotation=<rot 90 0 0>}]
 dyn_hair_back_l_01 [{shapeType=Capsule, length=1, radius=1, center=<0.1 0 0>, rotation=<rot 90 0 0>}]
 dyn_hair_back_l_02 [{shapeType=Capsule, length=1, radius=1, center=<0.1 0 0>, rotation=<rot 90 0 0>}]
 dyn_hair_back_r_01 [{shapeType=Capsule, length=1.5, radius=1, center=<0.1 0 0>, rotation=<rot 90 0 0>}]
 dyn_hair_back_r_02 [{shapeType=Capsule, length=2.6124, radius=1.0794, center=<1.9642 0.8132 0.0001>, rotation=<rot 90 0 -12.5403>}]
 dyn_hair_l_01 [{shapeType=Capsule, length=1, radius=1, center=<0.1 0 0>, rotation=<rot 90 0 0>}]
 dyn_hair_l_02 [{shapeType=Capsule, length=1, radius=1, center=<0.1 0 0>, rotation=<rot 90 0 0>}]
 dyn_hair_l_03 [{shapeType=Capsule, length=1, radius=1, center=<0.1 0 0>, rotation=<rot 90 0 0>}]
 dyn_hair_mid_l_01 [{shapeType=Capsule, length=1, radius=1, center=<0.1 0 0>, rotation=<rot 90 0 0>}]
 dyn_hair_mid_l_02 [{shapeType=Capsule, length=1, radius=1, center=<0.1 0 0>, rotation=<rot 90 0 0>}]
 dyn_hair_mid_l_03 [{shapeType=Capsule, length=1, radius=1, center=<0.1 0 0>, rotation=<rot 90 0 0>}]
 dyn_hair_mid_l_04 [{shapeType=Capsule, length=1, radius=1, center=<0.1 0 0>, rotation=<rot 90 0 0>}]
 dyn_hair_mid_r_01 [{shapeType=Capsule, length=1.105, radius=1.002, center=<0.4907 0.0736 -0.4994>, rotation=<rot 90 0 0>}]
 dyn_hair_mid_r_02 [{shapeType=Capsule, length=4.0091, radius=2.0063, center=<1.765 -0.4263 0.9168>, rotation=<rot 44.2546 176.8703 -179.0526>}]
 dyn_hair_r_01 [{shapeType=Capsule, length=2, radius=1.25, center=<0.1 0 0>, rotation=<rot 90 0 0>}]
 dyn_hair_r_02 [{shapeType=Capsule, length=2, radius=1.25, center=<0.1 0 0>, rotation=<rot 90 0 0>}]
 dyn_hair_r_03 [{shapeType=Capsule, length=2, radius=1.25, center=<0.1 0 0>, rotation=<rot 90 0 0>}]
 dyn_headband_l_01 [{shapeType=Box, extentX=2.5, extentY=0.5, extentZ=1.5, center=<1 0 0>}]
 dyn_headband_l_02 [{shapeType=Box, extentX=2.5, extentY=0.5, extentZ=1.5, center=<1 0 0>}]
 dyn_headband_r_01 [{shapeType=Box, extentX=2.5, extentY=0.5, extentZ=1.5, center=<1 0 0>}]
 dyn_headband_r_02 [{shapeType=Box, extentX=2.5, extentY=0.5, extentZ=1.5, center=<1 0 0>}]
 dyn_headband_r_03 [{shapeType=Box, extentX=2.5, extentY=0.5, extentZ=1.5, center=<1 0 0>}]
 head [{shapeType=Capsule, length=6.1269, radius=6.6248, center=<4.2824 1.7124 0.751>, rotation=<rot 0 0 -90.0542>}]
}
constraints [
 {bone1Name=dyn_bun_l, bone2Name=head, swing1LimitDegrees=45, swing2LimitDegrees=45, twistLimitDegrees=15}
 {bone1Name=dyn_bun_r, bone2Name=head, swing1LimitDegrees=45, swing2LimitDegrees=45, twistLimitDegrees=15}
 {bone1Name=dyn_hair_r_01, bone2Name=head, swing1LimitDegrees=20, swing2LimitDegrees=20, twistLimitDegrees=5}
 {bone1Name=dyn_hair_r_02, bone2Name=dyn_hair_r_01, swing1LimitDegrees=45, swing2LimitDegrees=45, twistLimitDegrees=15}
 {bone1Name=dyn_hair_r_03, bone2Name=dyn_hair_r_02, swing1LimitDegrees=45, swing2LimitDegrees=45, twistLimitDegrees=15}
 {bone1Name=dyn_hair_mid_r_01, bone2Name=dyn_hair_r_01, swing1LimitDegrees=30, swing2LimitDegrees=30, twistLimitDegrees=15}
 {bone1Name=dyn_hair_mid_r_02, bone2Name=dyn_hair_mid_r_01, swing1LimitDegrees=30, swing2LimitDegrees=30, twistLimitDegrees=15}
 {bone1Name=dyn_headband_r_01, bone2Name=head, swing1LimitDegrees=45, swing2LimitDegrees=45, twistLimitDegrees=15}
 {bone1Name=dyn_headband_r_02, bone2Name=dyn_headband_r_01, swing1LimitDegrees=45, swing2LimitDegrees=45, twistLimitDegrees=15}
 {bone1Name=dyn_headband_r_03, bone2Name=dyn_headband_r_02, swing1LimitDegrees=45, swing2LimitDegrees=45, twistLimitDegrees=15}
 {bone1Name=dyn_headband_l_01, bone2Name=head, swing1LimitDegrees=45, swing2LimitDegrees=45, twistLimitDegrees=15}
 {bone1Name=dyn_headband_l_02, bone2Name=dyn_headband_l_01, swing1LimitDegrees=45, swing2LimitDegrees=45, twistLimitDegrees=15}
 {bone1Name=dyn_hair_back_r_01, bone2Name=head, swing1LimitDegrees=45, swing2LimitDegrees=45, twistLimitDegrees=15}
 {bone1Name=dyn_hair_back_r_02, bone2Name=dyn_hair_back_r_01, swing1LimitDegrees=45, swing2LimitDegrees=45, twistLimitDegrees=15}
 {bone1Name=dyn_hair_l_01, bone2Name=head, swing1LimitDegrees=45, swing2LimitDegrees=45, twistLimitDegrees=15}
 {bone1Name=dyn_hair_l_02, bone2Name=dyn_hair_l_01, swing1LimitDegrees=45, swing2LimitDegrees=45, twistLimitDegrees=15}
 {bone1Name=dyn_hair_l_03, bone2Name=dyn_hair_l_02, swing1LimitDegrees=45, swing2LimitDegrees=45, twistLimitDegrees=15}
 {bone1Name=dyn_hair_mid_l_01, bone2Name=head, swing1LimitDegrees=30, swing2LimitDegrees=30, twistLimitDegrees=15}
 {bone1Name=dyn_hair_mid_l_02, bone2Name=dyn_hair_mid_l_01, swing1LimitDegrees=45, swing2LimitDegrees=45, twistLimitDegrees=15}
 {bone1Name=dyn_hair_mid_l_03, bone2Name=dyn_hair_mid_l_02, swing1LimitDegrees=45, swing2LimitDegrees=45, twistLimitDegrees=15}
 {bone1Name=dyn_hair_mid_l_04, bone2Name=dyn_hair_mid_l_03, swing1LimitDegrees=45, swing2LimitDegrees=45, twistLimitDegrees=15}
 {bone1Name=dyn_hair_back_l_01, bone2Name=dyn_hair_mid_l_01, swing1LimitDegrees=45, swing2LimitDegrees=45, twistLimitDegrees=15}
 {bone1Name=dyn_hair_back_l_02, bone2Name=dyn_hair_back_l_01, swing1LimitDegrees=45, swing2LimitDegrees=45, twistLimitDegrees=15}
]
