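# Physics asset for the buns and tassel part of the female medium Red Riding Remix character body, simulating dangling headband pieces.
# Small box bodies form two chains hanging from the head, three bones on the right and two on the left.
# Each link shares the same limited swing and tighter twist constraint.
# A head capsule and an unconstrained right clavicle box add body volumes, likely as colliders for the chains.
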PhysicsAsset /BRCosmetics/Characters/Player/Female/Medium/Bodies/F_MED_RedRidingRemix/Meshes/Parts/F_MED_Red_Riding_Remix_Buns_Tassle_Physics
bodies {
 clavicle_r [{shapeType=Box, extentX=3.51, extentY=18.1211, extentZ=9.6412, center=<-10.2731 -3.062 -9.8429>}]
 dyn_headband_l_01 [{shapeType=Box, extentX=2.5, extentY=0.5, extentZ=1.5, center=<1 0 0>}]
 dyn_headband_l_02 [{shapeType=Box, extentX=2.5, extentY=0.5, extentZ=1.5, center=<1 0 0>}]
 dyn_headband_r_01 [{shapeType=Box, extentX=2.5, extentY=0.5, extentZ=1.5, center=<1 0 0>}]
 dyn_headband_r_02 [{shapeType=Box, extentX=2.5, extentY=0.5, extentZ=1.5, center=<1 0 0>}]
 dyn_headband_r_03 [{shapeType=Box, extentX=2.5, extentY=0.5, extentZ=1.5, center=<1 0 0>}]
 head [{shapeType=Capsule, length=6.1269, radius=6.6248, center=<4.2824 1.7124 0.751>, rotation=<rot 0 0 -90.0542>}]
}
constraints [
 {bone1Name=dyn_headband_r_01, bone2Name=head, swing1LimitDegrees=45, swing2LimitDegrees=45, twistLimitDegrees=15}
 {bone1Name=dyn_headband_r_02, bone2Name=dyn_headband_r_01, swing1LimitDegrees=45, swing2LimitDegrees=45, twistLimitDegrees=15}
 {bone1Name=dyn_headband_r_03, bone2Name=dyn_headband_r_02, swing1LimitDegrees=45, swing2LimitDegrees=45, twistLimitDegrees=15}
 {bone1Name=dyn_headband_l_01, bone2Name=head, swing1LimitDegrees=45, swing2LimitDegrees=45, twistLimitDegrees=15}
 {bone1Name=dyn_headband_l_02, bone2Name=dyn_headband_l_01, swing1LimitDegrees=45, swing2LimitDegrees=45, twistLimitDegrees=15}
]
