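# Physics asset for the Red Riding Remix female medium cosmetic's face accessory, covering its braids, wolf ears, nose, hair pieces and back buckle.
# Two five-link braid chains, the ears, nose and six hair boxes hang from the head with limited swing and twist, tighter on the hair.
# The buckle attaches to neck_02.
# Head, neck, clavicle and spine_05 bodies form the upper-body base, linked by free angular constraints.
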PhysicsAsset /BRCosmetics/Characters/Player/Female/Medium/Bodies/F_MED_RedRidingRemix/Meshes/Parts/F_MED_Red_Riding_Remix_FaceAcc_Physics
bodies {
 clavicle_l [{shapeType=Capsule, length=10.749, radius=6.8701, center=<4.25 0 0>, rotation=<rot 90 0 0>}]
 clavicle_r [{shapeType=Capsule, length=10.749, radius=6.8701, center=<-4.25 0 0>, rotation=<rot 90 0 0>}]
 dyn_braid_l_01 [{shapeType=Capsule, length=3, radius=2, center=<2 0 0>, rotation=<rot 90 0 0>}]
 dyn_braid_l_02 [{shapeType=Capsule, length=3, radius=2, center=<2 0 0>, rotation=<rot 90 0 0>}]
 dyn_braid_l_03 [{shapeType=Capsule, length=3, radius=2, center=<2 0 0>, rotation=<rot 90 0 0>}]
 dyn_braid_l_04 [{shapeType=Capsule, length=3, radius=2, center=<2 0 0>, rotation=<rot 90 0 0>}]
 dyn_braid_l_05 [{shapeType=Capsule, length=3, radius=2, center=<2 0 0>, rotation=<rot 90 0 0>}]
 dyn_braid_r_01 [{shapeType=Capsule, length=3, radius=2, center=<2 0 0>, rotation=<rot 90 0 0>}]
 dyn_braid_r_02 [{shapeType=Capsule, length=3, radius=2, center=<2 0 0>, rotation=<rot 90 0 0>}]
 dyn_braid_r_03 [{shapeType=Capsule, length=3, radius=2, center=<2 0 0>, rotation=<rot 90 0 0>}]
 dyn_braid_r_04 [{shapeType=Capsule, length=3, radius=2, center=<2 0 0>, rotation=<rot 90 0 0>}]
 dyn_braid_r_05 [{shapeType=Capsule, length=3, radius=2, center=<2 0 0>, rotation=<rot 90 0 0>}]
 dyn_buckle_back [{shapeType=Box, extentX=3, extentY=0.4, extentZ=4, center=<1 0 0>}]
 dyn_wolf_ear_l [{shapeType=Sphere, radius=3, center=<1 0 0>}]
 dyn_wolf_ear_r [{shapeType=Sphere, radius=3, center=<1 0 0>}]
 dyn_wolf_hair_01 [{shapeType=Box, extentX=2.168, extentY=0.5001, extentZ=3.1455, center=<0.1 0 0>}]
 dyn_wolf_hair_02 [{shapeType=Box, extentX=2.168, extentY=0.5001, extentZ=3.1455, center=<0.1 0 0>}]
 dyn_wolf_hair_03 [{shapeType=Box, extentX=2.168, extentY=0.5001, extentZ=3.1455, center=<0.1 0 0>}]
 dyn_wolf_hair_04 [{shapeType=Box, extentX=2.168, extentY=0.5001, extentZ=3.1455, center=<0.1 -0.14 0>}]
 dyn_wolf_hair_05 [{shapeType=Box, extentX=2.168, extentY=0.5001, extentZ=3.1455, center=<0.1 0 0>}]
 dyn_wolf_hair_06 [{shapeType=Box, extentX=2.168, extentY=0.5001, extentZ=3.1455, center=<0.1 0 0>}]
 dyn_wolf_nose [{shapeType=Capsule, length=4.3525, radius=4.6166, center=<0.1 0 0>, rotation=<rot 90 0 0>}]
 head [{shapeType=Sphere, radius=4.0194, center=<6.9152 6.2113 0>}]
 neck_02 [{shapeType=Capsule, length=7.9707, radius=6.9798, center=<-4.5034 -1.1854 0>, rotation=<rot 0 179.9998 -180>}]
 spine_05 [
  {shapeType=Capsule, length=9.0795, radius=5.7357, center=<-2.7445 9.0845 0.1776>, rotation=<rot 0.0003 50.0042 -0.3256>}
  {shapeType=Box, extentX=18.8932, extentY=7.6708, extentZ=17.4222, center=<3.9035 4.3582 0>, rotation=<rot 0 -41.8489 0>}
 ]
}
constraints [
 {bone1Name=dyn_braid_l_01, bone2Name=head, swing1LimitDegrees=45, swing2LimitDegrees=45, twistLimitDegrees=15}
 {bone1Name=dyn_braid_r_01, bone2Name=head, swing1LimitDegrees=45, swing2LimitDegrees=45, twistLimitDegrees=15}
 {
  bone1Name=clavicle_l
  bone2Name=spine_05
  swing1LimitDegrees=45
  swing1Motion=Free
  swing2LimitDegrees=45
  swing2Motion=Free
  twistLimitDegrees=45
  twistMotion=Free
 }
 {
  bone1Name=clavicle_r
  bone2Name=spine_05
  swing1LimitDegrees=45
  swing1Motion=Free
  swing2LimitDegrees=45
  swing2Motion=Free
  twistLimitDegrees=45
  twistMotion=Free
 }
 {
  bone1Name=neck_02
  bone2Name=spine_05
  swing1LimitDegrees=45
  swing1Motion=Free
  swing2LimitDegrees=45
  swing2Motion=Free
  twistLimitDegrees=45
  twistMotion=Free
 }
 {bone1Name=dyn_braid_l_02, bone2Name=dyn_braid_l_01, swing1LimitDegrees=45, swing2LimitDegrees=45, twistLimitDegrees=15}
 {bone1Name=dyn_braid_l_03, bone2Name=dyn_braid_l_02, swing1LimitDegrees=45, swing2LimitDegrees=45, twistLimitDegrees=15}
 {bone1Name=dyn_braid_l_04, bone2Name=dyn_braid_l_03, swing1LimitDegrees=45, swing2LimitDegrees=45, twistLimitDegrees=15}
 {bone1Name=dyn_braid_l_05, bone2Name=dyn_braid_l_04, swing1LimitDegrees=45, swing2LimitDegrees=45, twistLimitDegrees=15}
 {bone1Name=dyn_braid_r_02, bone2Name=dyn_braid_r_01, swing1LimitDegrees=45, swing2LimitDegrees=45, twistLimitDegrees=15}
 {bone1Name=dyn_braid_r_04, bone2Name=dyn_braid_r_03, swing1LimitDegrees=45, swing2LimitDegrees=45, twistLimitDegrees=15}
 {bone1Name=dyn_braid_r_03, bone2Name=dyn_braid_r_02, swing1LimitDegrees=45, swing2LimitDegrees=45, twistLimitDegrees=15}
 {bone1Name=dyn_braid_r_05, bone2Name=dyn_braid_r_04, swing1LimitDegrees=45, swing2LimitDegrees=45, twistLimitDegrees=15}
 {bone1Name=dyn_wolf_hair_01, bone2Name=head, swing1LimitDegrees=30, swing2LimitDegrees=30, twistLimitDegrees=10}
 {bone1Name=dyn_wolf_hair_02, bone2Name=head, swing1LimitDegrees=30, swing2LimitDegrees=30, twistLimitDegrees=10}
 {bone1Name=dyn_wolf_hair_03, bone2Name=head, swing1LimitDegrees=30, swing2LimitDegrees=30, twistLimitDegrees=10}
 {bone1Name=dyn_wolf_hair_04, bone2Name=head, swing1LimitDegrees=30, swing2LimitDegrees=30, twistLimitDegrees=10}
 {bone1Name=dyn_wolf_hair_05, bone2Name=head, swing1LimitDegrees=30, swing2LimitDegrees=30, twistLimitDegrees=10}
 {bone1Name=dyn_wolf_hair_06, bone2Name=head, swing1LimitDegrees=30, swing2LimitDegrees=30, twistLimitDegrees=10}
 {bone1Name=dyn_wolf_nose, bone2Name=head, swing1LimitDegrees=45, swing2LimitDegrees=45, twistLimitDegrees=15}
 {bone1Name=dyn_wolf_ear_l, bone2Name=head, swing1LimitDegrees=45, swing2LimitDegrees=45, twistLimitDegrees=15}
 {bone1Name=dyn_wolf_ear_r, bone2Name=head, swing1LimitDegrees=45, swing2LimitDegrees=45, twistLimitDegrees=15}
 {
  bone1Name=head
  bone2Name=neck_02
  swing1LimitDegrees=45
  swing1Motion=Free
  swing2LimitDegrees=45
  swing2Motion=Free
  twistLimitDegrees=45
  twistMotion=Free
 }
 {bone1Name=dyn_buckle_back, bone2Name=neck_02, swing1LimitDegrees=45, swing2LimitDegrees=45, twistLimitDegrees=15}
]
