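# Control Rig for the Red Riding Remix female face accessory, named as an LOD fix, that re-aims the left and right braid bones.
# Two Forward Solve chains, one per side, rotate the 02 then 03 braid bone so its X axis points at a fixed offset from spine_05.
# Each uses a from-two-vectors quaternion multiplied onto the current global rotation, propagated to children.
# Debug axes remain enabled at each 01 bone and first target.
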ControlRigBlueprint /BRCosmetics/Characters/Player/Female/Medium/Bodies/F_MED_RedRidingRemix/Meshes/Parts/Red_Remix_FaceAcc_LODFix_CtrlRig
graphs [
 RigVMFunctionLibrary {links [], nodes []}
 RigVMModel {nodes [
   n1 = &RigUnit_BeginExecution()
   /Script/ControlRig.RigUnit_DebugTransformMutable(
    Color=(R=1,G=0,B=0,A=1)
    DebugDrawSettings=(DepthPriority=SDPG_Foreground,Lifetime=-1)
    Mode=Axes
    Scale=1
    Space=dyn_braid_l_01
    Thickness=0
    Transform=(Rotation=(X=0,Y=0,Z=0,W=1),Translation=(X=0,Y=0,Z=0),Scale3D=(X=1,Y=1,Z=1))
    WorldOffset=(Rotation=(X=0,Y=0,Z=0,W=1),Translation=(X=0,Y=0,Z=0),Scale3D=(X=1,Y=1,Z=1))
    bEnabled=false
    ExecutePin=@n1.ExecutePin
   ) {is=struct RigVMUnitNode}
   n2 = &RigUnit_GetBoneTransform(Bone=dyn_braid_l_02)
   n3 = /Script/RigVM.RigVMFunction_MathQuaternionGetAxis(Axis=X, Quaternion=@n2.Transform.Rotation) {is=struct RigVMUnitNode, template="Axis::Execute(in Quaternion,in Axis,out Result)"}
   n4 = /Script/RigVM.RigVMFunction_MathTransformTransformVector(
    Transform=&RigUnit_GetBoneTransform(Bone=spine_05).Transform
    Location=/Script/RigVM.RigVMFunction_MathVectorAdd(A=(X=0,Y=15,Z=-7), B=(X=0,Y=0,Z=0)) {is=struct RigVMUnitNode, template="Add::Execute(in A,in B,out Result)"}.Result
   ) {is=struct RigVMUnitNode}
   n5 = /Script/RigVM.RigVMFunction_MathVectorSub(B=@n2.Transform.Translation, A=@n4.Result) {is=struct RigVMUnitNode, template="Subtract::Execute(in A,in B,out Result)"}
   > /Script/ControlRig.RigUnit_SetBoneRotation(
    Bone=dyn_braid_l_02
    Space=GlobalSpace
    Weight=1
    bPropagateToChildren=True
    Rotation=/Script/RigVM.RigVMFunction_MathQuaternionMul(
     B=@n2.Transform.Rotation
     A=/Script/RigVM.RigVMFunction_MathQuaternionFromTwoVectors(A=@n3.Result, B=@n5.Result) {is=struct RigVMUnitNode, template="FromTwoVectors::Execute(in A,in B,out Result)"}.Result
    ) {is=struct RigVMUnitNode, template="Multiply::Execute(in A,in B,out Result)"}.Result
   ) {is=struct RigVMUnitNode}
   > &RigUnit_DebugTransformMutable(
    Color=(R=1,G=0,B=0,A=1)
    DebugDrawSettings=(DepthPriority=SDPG_Foreground,Lifetime=-1)
    Mode=Axes
    Scale=75
    Space=dyn_braid_l_01
    Thickness=0
    Transform=(Rotation=(X=0,Y=0,Z=0,W=1),Translation=(X=0,Y=0,Z=0),Scale3D=(X=1,Y=1,Z=1))
    WorldOffset=(Rotation=(X=0,Y=0,Z=0,W=1),Translation=(X=0,Y=0,Z=0),Scale3D=(X=1,Y=1,Z=1))
    bEnabled=True
   )
   > &RigUnit_DebugTransformMutable(
    Color=(R=1,G=0,B=0,A=1)
    DebugDrawSettings=(DepthPriority=SDPG_Foreground,Lifetime=-1)
    Mode=Axes
    Scale=55
    Space=None
    Thickness=0
    Transform=(Rotation=(X=0,Y=0,Z=0,W=1),Translation=(X=0,Y=0,Z=0),Scale3D=(X=1,Y=1,Z=1))
    WorldOffset=(Rotation=(X=0,Y=0,Z=0,W=1),Translation=(X=0,Y=0,Z=0),Scale3D=(X=1,Y=1,Z=1))
    bEnabled=True
    Transform.Translation=@n4.Result
   )
   n6 = &RigUnit_GetBoneTransform(Bone=dyn_braid_l_03)
   n7 = &RigVMFunction_MathQuaternionGetAxis(Axis=X, Quaternion=@n6.Transform.Rotation) {template="Axis::Execute(in Quaternion,in Axis,out Result)"}
   n8 = &RigVMFunction_MathVectorSub(
    B=@n6.Transform.Translation
    A=&RigVMFunction_MathTransformTransformVector(
     Transform=&RigUnit_GetBoneTransform(Bone=spine_05).Transform
     Location=&RigVMFunction_MathVectorAdd(A=(X=0,Y=14,Z=-7), B=(X=0,Y=0,Z=0)) {template="Add::Execute(in A,in B,out Result)"}.Result
    ).Result
   ) {template="Subtract::Execute(in A,in B,out Result)"}
   > &RigUnit_SetBoneRotation(
    Bone=dyn_braid_l_03
    Space=GlobalSpace
    Weight=1
    bPropagateToChildren=True
    Rotation=&RigVMFunction_MathQuaternionMul(
     B=@n6.Transform.Rotation
     A=&RigVMFunction_MathQuaternionFromTwoVectors(A=@n7.Result, B=@n8.Result) {template="FromTwoVectors::Execute(in A,in B,out Result)"}.Result
    ) {template="Multiply::Execute(in A,in B,out Result)"}.Result
   )
   n9 = &RigUnit_BeginExecution()
   &RigUnit_DebugTransformMutable(
    Color=(R=1,G=0,B=0,A=1)
    DebugDrawSettings=(DepthPriority=SDPG_Foreground,Lifetime=-1)
    Mode=Axes
    Scale=1
    Space=dyn_braid_r_01
    Thickness=0
    Transform=(Rotation=(X=0,Y=0,Z=0,W=1),Translation=(X=0,Y=0,Z=0),Scale3D=(X=1,Y=1,Z=1))
    WorldOffset=(Rotation=(X=0,Y=0,Z=0,W=1),Translation=(X=0,Y=0,Z=0),Scale3D=(X=1,Y=1,Z=1))
    bEnabled=false
    ExecutePin=@n9.ExecutePin
   )
   n10 = &RigUnit_GetBoneTransform(Bone=dyn_braid_r_02)
   n11 = &RigVMFunction_MathQuaternionGetAxis(Axis=X, Quaternion=@n10.Transform.Rotation) {template="Axis::Execute(in Quaternion,in Axis,out Result)"}
   n12 = &RigVMFunction_MathTransformTransformVector(
    Transform=&RigUnit_GetBoneTransform(Bone=spine_05).Transform
    Location=&RigVMFunction_MathVectorAdd(A=(X=0,Y=15,Z=7), B=(X=0,Y=0,Z=0)) {template="Add::Execute(in A,in B,out Result)"}.Result
   )
   n13 = &RigVMFunction_MathVectorSub(B=@n10.Transform.Translation, A=@n12.Result) {template="Subtract::Execute(in A,in B,out Result)"}
   > &RigUnit_SetBoneRotation(
    Bone=dyn_braid_r_02
    Space=GlobalSpace
    Weight=1
    bPropagateToChildren=True
    Rotation=&RigVMFunction_MathQuaternionMul(
     B=@n10.Transform.Rotation
     A=&RigVMFunction_MathQuaternionFromTwoVectors(A=@n11.Result, B=@n13.Result) {template="FromTwoVectors::Execute(in A,in B,out Result)"}.Result
    ) {template="Multiply::Execute(in A,in B,out Result)"}.Result
   )
   > &RigUnit_DebugTransformMutable(
    Color=(R=1,G=0,B=0,A=1)
    DebugDrawSettings=(DepthPriority=SDPG_Foreground,Lifetime=-1)
    Mode=Axes
    Scale=75
    Space=dyn_braid_r_01
    Thickness=0
    Transform=(Rotation=(X=0,Y=0,Z=0,W=1),Translation=(X=0,Y=0,Z=0),Scale3D=(X=1,Y=1,Z=1))
    WorldOffset=(Rotation=(X=0,Y=0,Z=0,W=1),Translation=(X=0,Y=0,Z=0),Scale3D=(X=1,Y=1,Z=1))
    bEnabled=True
   )
   > &RigUnit_DebugTransformMutable(
    Color=(R=1,G=0,B=0,A=1)
    DebugDrawSettings=(DepthPriority=SDPG_Foreground,Lifetime=-1)
    Mode=Axes
    Scale=55
    Space=None
    Thickness=0
    Transform=(Rotation=(X=0,Y=0,Z=0,W=1),Translation=(X=0,Y=0,Z=0),Scale3D=(X=1,Y=1,Z=1))
    WorldOffset=(Rotation=(X=0,Y=0,Z=0,W=1),Translation=(X=0,Y=0,Z=0),Scale3D=(X=1,Y=1,Z=1))
    bEnabled=True
    Transform.Translation=@n12.Result
   )
   n14 = &RigUnit_GetBoneTransform(Bone=dyn_braid_r_03)
   n15 = &RigVMFunction_MathQuaternionGetAxis(Axis=X, Quaternion=@n14.Transform.Rotation) {template="Axis::Execute(in Quaternion,in Axis,out Result)"}
   n16 = &RigVMFunction_MathVectorSub(
    B=@n14.Transform.Translation
    A=&RigVMFunction_MathTransformTransformVector(
     Transform=&RigUnit_GetBoneTransform(Bone=spine_05).Transform
     Location=&RigVMFunction_MathVectorAdd(A=(X=0,Y=14,Z=7), B=(X=0,Y=0,Z=0)) {template="Add::Execute(in A,in B,out Result)"}.Result
    ).Result
   ) {template="Subtract::Execute(in A,in B,out Result)"}
   > &RigUnit_SetBoneRotation(
    Bone=dyn_braid_r_03
    Space=GlobalSpace
    Weight=1
    bPropagateToChildren=True
    Rotation=&RigVMFunction_MathQuaternionMul(
     B=@n14.Transform.Rotation
     A=&RigVMFunction_MathQuaternionFromTwoVectors(A=@n15.Result, B=@n16.Result) {template="FromTwoVectors::Execute(in A,in B,out Result)"}.Result
    ) {template="Multiply::Execute(in A,in B,out Result)"}.Result
   )
  ]}
]
hierarchy [
 root {loc=<0 0 0>, rot=<0 0 0 1>}
 attach {parent=root, loc=<0 0 0>, rot=<0 0 0 1>}
 pelvis {parent=root, loc=<0 -0.5301 98.0571>, rot=<0.0127 0.707 -0.0127 -0.707>}
 spine_01 {parent=pelvis, loc=<6.1208 -0.9328 0>, rot=<0 0 0.034 -0.9994>}
 spine_02 {parent=spine_01, loc=<6.5996 -0.2095 0>, rot=<0 0 0.0238 -0.9997>}
 spine_03 {parent=spine_02, loc=<6.4914 -0.1237 0>, rot=<0 0 0.0149 -0.9999>}
 spine_04 {parent=spine_03, loc=<6.9598 0.0794 0>, rot=<0 0 -0.0132 -0.9999>}
 spine_05 {parent=spine_04, loc=<6.9987 0 0>, rot=<0 0 -0.0015 -1>}
 clavicle_l {parent=spine_05, loc=<5.4831 -0.1053 -3.2946>, rot=<0.0465 -0.7289 -0.009 -0.6829>}
 upperarm_l {parent=clavicle_l, loc=<9.8658 -1.2136 -0.4105>, rot=<0.0164 -0.3245 -0.0269 -0.9454>}
 clavicle_r {parent=spine_05, loc=<5.4831 -0.1053 3.2946>, rot=<0.7289 0.0465 -0.6829 0.009>}
 upperarm_r {parent=clavicle_r, loc=<-9.8658 1.2136 0.4104>, rot=<0.0164 -0.3245 -0.0269 -0.9454>}
 neck_01 {parent=spine_05, loc=<11.3138 0.1529 0>, rot=<0 0 -0.2003 -0.9797>}
 neck_02 {parent=neck_01, loc=<4.5627 0 0>, rot=<0 0 0.0504 -0.9987>}
 head {parent=neck_02, loc=<5.3914 0 0>, rot=<0 0 0.115 -0.9934>}
 dyn_wolf_nose {parent=head, loc=<10.4482 10.9348 0>, rot=<0 0 -0.8558 -0.5174>}
 dyn_wolf_ear_l {parent=head, loc=<16.2008 2.7626 -6.0547>, rot=<0.1198 -0.3767 0.375 -0.8385>}
 dyn_wolf_ear_r {parent=head, loc=<16.2008 2.7626 6.055>, rot=<0.198 0.2552 0.4604 -0.8269>}
 dyn_wolf_hair_01 {parent=head, loc=<19.819 5.3906 0>, rot=<0 0 0.2301 -0.9732>}
 dyn_wolf_hair_02 {parent=head, loc=<19.9055 2.4912 0>, rot=<0 0 0.5002 -0.8659>}
 dyn_wolf_hair_03 {parent=head, loc=<18.5002 -2.3998 0>, rot=<0 0 0.599 -0.8008>}
 dyn_wolf_hair_04 {parent=head, loc=<16.6369 -3.9922 0>, rot=<0 0 0.8083 -0.5888>}
 dyn_wolf_hair_05 {parent=head, loc=<12.9178 -6.237 0>, rot=<0 0 0.8435 -0.5372>}
 dyn_wolf_hair_06 {parent=head, loc=<9.526 -7.7649 0>, rot=<0 0 0.8775 -0.4795>}
 dyn_braid_l_01 {parent=head, loc=<-3.2695 1.655 -7.1957>, rot=<0 0 0.9177 0.3972>}
 dyn_braid_l_02 {parent=dyn_braid_l_01, loc=<7.5153 0 0>, rot=<0 0 0 1>}
 dyn_braid_l_03 {parent=dyn_braid_l_02, loc=<8.5392 0 0>, rot=<0 0 0 1>}
 dyn_braid_l_04 {parent=dyn_braid_l_03, loc=<6.7832 0 0>, rot=<0 0 0 1>}
 dyn_braid_l_05 {parent=dyn_braid_l_04, loc=<7.0738 0 0>, rot=<0 0 0 1>}
 dyn_braid_r_01 {parent=head, loc=<-3.2695 1.655 7.196>, rot=<0 0 0.9177 0.3972>}
 dyn_braid_r_02 {parent=dyn_braid_r_01, loc=<7.5153 0 0>, rot=<0 0 0 1>}
 dyn_braid_r_03 {parent=dyn_braid_r_02, loc=<8.5392 0 0>, rot=<0 0 0 1>}
 dyn_braid_r_04 {parent=dyn_braid_r_03, loc=<6.7832 0 0>, rot=<0 0 0 1>}
 dyn_braid_r_05 {parent=dyn_braid_r_04, loc=<7.0738 0 0>, rot=<0 0 0 1>}
 dyn_buckle_back {parent=neck_02, loc=<2.2987 -11.4306 0>, rot=<0 0 0.9942 -0.1078>}
 dyn_fur_back {parent=spine_05, loc=<2.0357 -9.2932 0>, rot=<0.0517 -0.9987 0 0>}
]
