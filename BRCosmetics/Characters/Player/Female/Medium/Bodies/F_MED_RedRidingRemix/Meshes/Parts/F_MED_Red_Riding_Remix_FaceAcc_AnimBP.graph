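# Anim Blueprint for the Red Riding Remix face accessory mesh on a female player character, driving secondary motion of braids, buckle and fur.
# It copies the attached parent's pose, applies a Control Rig and braid offsets, then runs buckle AnimDynamics and RigidBody simulation.
# A LOD-fix Control Rig replaces that simulation when rigid bodies are disabled or LOD exceeds 1.
# Back fur AnimDynamics runs only without a backpack, detected on customization load.
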AnimBlueprint /BRCosmetics/Characters/Player/Female/Medium/Bodies/F_MED_RedRidingRemix/Meshes/Parts/F_MED_Red_Riding_Remix_FaceAcc_AnimBP
target_skeleton=&/F_MED_Red_Riding_Remix_FaceAcc_Skeleton
graphs [
 AnimGraph {nodes [
   n1 = Anim Dynamics - Bone: dyn_fur_back(ComponentPose=Local To Component(LocalPose=Use cached pose 'RBPose'()), Alpha=1, ExternalForce="0,0,0") {
    angularDampingOverride=0.98
    bGravityOverrideInSimSpace=true
    bOverrideAngularDamping=true
    bUseGravityOverride=true
    alphaBoolBlend {blendOption=Linear}
    boundBone {boneName=dyn_fur_back}
    componentLinearAccScale=<1 1>
    gravityOverride {z=-980}
    physicsBodyDefinitions [{
      sphereCollisionRadius=0
      boundBone {boneName=dyn_fur_back}
      boxExtents=<20 5 20>
      constraintSetup {angularLimitsMax=<1 5 5>, angularLimitsMin=<-1 -5 -50>, angularTarget=<0 0 0>}
      localJointOffset=<5 0 0>
     }]
    relativeSpaceBone {boneName=spine_05}
   }
   Output Pose(Result=Blend Poses by bool(
     BlendPose_0=Use cached pose 'RBPose'()
     BlendPose_1=Component To Local(ComponentPose=@n1)
     bActiveValue=Get IsBackpackEquipped() -> (IsBackpackEquipped:bool)
    ) {blendType=Linear}) {layerGroup=None}
   n2 = Control Rig(Source=Copy Pose From Mesh() {bUseAttachedParent=true}) {
    alphaBoolBlend {blendOption=Linear}
    controlRigAssetReference {blueprintRigClass=&/Red_Remix_FaceAcc_CtrlRig.Red_Remix_FaceAcc_CtrlRig_C}
   }
   n3 = "Transform (Modify) Bone - Bone: dyn_braid_l_01"(ComponentPose=Local To Component(LocalPose=@n2), Alpha=1, Translation="-2,0,0", Rotation="0,0,-20", Scale="1,1,1") {
    rotationMode=BMM_Additive
    translationMode=BMM_Additive
    alphaBoolBlend {blendOption=Linear}
    boneToModify {boneName=dyn_braid_l_01}
    rotation {roll=-20}
    translation {x=-2}
   }
   n4 = "Transform (Modify) Bone - Bone: dyn_braid_r_01"(ComponentPose=@n3, Alpha=1, Translation="2,0,0", Rotation="0.000,0.000,-25", Scale="1,1,1") {
    rotationMode=BMM_Additive
    translationMode=BMM_Additive
    alphaBoolBlend {blendOption=Linear}
    boneToModify {boneName=dyn_braid_r_01}
    rotation {roll=-25}
    translation {x=2}
   }
   Save cached pose 'BasePose'(Pose=Component To Local(ComponentPose=@n4)) {cacheName=BasePose}
   n5 = Anim Dynamics - Bone: dyn_buckle_back(ComponentPose=Local To Component(LocalPose=Use cached pose 'BasePose'()), Alpha=1, ExternalForce="0,0,0") {
    angularDampingOverride=0.98
    bGravityOverrideInSimSpace=true
    bOverrideAngularDamping=true
    bUseGravityOverride=true
    simulationSpace=RootRelative
    alphaBoolBlend {blendOption=Linear}
    boundBone {boneName=dyn_buckle_back}
    componentAppliedLinearAccClamp=<10000 10000 10000>
    componentLinearAccScale=<0.5 0.5 0.5>
    gravityOverride {z=-980}
    physicsBodyDefinitions [{
      sphereCollisionRadius=0
      boundBone {boneName=dyn_buckle_back}
      boxExtents=<5 1 5>
      constraintSetup {angularLimitsMax=<1 1 30>, angularLimitsMin=<-1 -1 -20>, angularTarget=<0 0 0>}
      localJointOffset=<1 0 0>
     }]
    relativeSpaceBone {boneName=root}
   }
   n6 = "Transform (Modify) Bone - Bone: dyn_braid_l_01"(ComponentPose=@n5, Alpha=1, Translation="0,-0.4,-0.000", Rotation="0,-13,0", Scale="1,1,1") {
    rotationMode=BMM_Additive
    rotationSpace=BCS_BoneSpace
    translationMode=BMM_Additive
    translationSpace=BCS_BoneSpace
    alphaBoolBlend {blendOption=Linear}
    boneToModify {boneName=dyn_braid_l_01}
    rotation {yaw=-13}
    translation {y=-0.4}
   }
   n7 = RigidBody(ComponentPose=@n6) {
    bOverrideWorldGravity=true
    overridePhysicsAsset=&/F_MED_Red_Riding_Remix_FaceAcc_Physics
    simulationSpace=BaseBoneSpace
    alphaBoolBlend {blendOption=Linear}
    baseBoneRef {boneName=spine_05}
    componentLinearAccScale=<0.25 0.25 0.25>
    overrideWorldGravity {z=-980}
   }
   n8 = Control Rig(Source=Use cached pose 'BasePose'()) {
    alphaBoolBlend {blendOption=Linear}
    controlRigAssetReference {blueprintRigClass=&/Red_Remix_FaceAcc_LODFix_CtrlRig.Red_Remix_FaceAcc_LODFix_CtrlRig_C}
   }
   Save cached pose 'RBPose'(Pose=Blend Poses by bool(BlendPose_0=@n8, BlendPose_1=Component To Local(ComponentPose=@n7), bActiveValue=Get isRigidBodyDisabled() -> (isRigidBodyDisabled:bool)) {blendType=Linear}) {cacheName=RBPose}
  ]}
 EventGraph {nodes [
   Event BlueprintUpdateAnimation()
   n1 = Get CurrentLOD()
   n2 = Get bRigidBodyEnabled()
   > Set isRigidBodyDisabled(isRigidBodyDisabled:bool=OR Boolean(A=NOT Boolean(A=@n2), B="integer > integer"(A=@n1, B=1) {is=title K2Node_CallFunction})) -> (Output_Get:bool)
   > Reroute Node()
   Event BlueprintInitializeAnimation()
   n3 = TryGetPawnOwner()
   > Is Valid(InputObject=@n3)
   > n4 = Cast To FortPlayerPawnAthena(Object=@n3)
   > Set PlayerPawn(PlayerPawn:Fort Player Pawn Athena Object Reference=@n4) -> (Output_Get:Fort Player Pawn Athena Object Reference)
   Event OnCharacterCustomizationLoaded()
   n5 = Get PlayerPawn() -> (PlayerPawn:Fort Player Pawn Athena Object Reference)
   > Is Valid(InputObject=@n5)
   > Do Once()
   n6 = CheckForBackbling()
   > Set isBackpackEquipped(isBackpackEquipped:bool=OR Boolean(A=@n6, B=@n6.bIsBackblingLong)) -> (Output_Get:bool)
   TryGetPawnOwner()
   "Not Equal (Object)"!(A:Object Reference=GetSkeletalMeshForPartType(self=@n5, PartType=Backpack), B:Object Reference) -> (ReturnValue:bool)
  ]}
]
variables [
 IsBackpackEquipped {type=boolean}
 isRigidBodyDisabled {type=boolean}
 PlayerPawn {type=FortPlayerPawnAthena}
]
